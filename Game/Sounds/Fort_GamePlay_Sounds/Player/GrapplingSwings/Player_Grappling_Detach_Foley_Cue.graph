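# Sound cue for the foley played when a player detaches during grappling swings.
# A Random node picks among five equally weighted detach foley waves.
# A root Modulator varies the pitch of the chosen wave to reduce repetition.
# The cue is assigned to the gadgets and items sound class.
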SoundCue /Game/Sounds/Fort_GamePlay_Sounds/Player/GrapplingSwings/Player_Grappling_Detach_Foley_Cue
root=Modulator_0
nodes [
 {id=Modulator_0, children [Random_1], properties {pitchMin=0.9}}
 {id=Random_1, children [WavePlayer_5,WavePlayer_6,WavePlayer_7,WavePlayer_8,WavePlayer_9], properties {weights [1,1,1,1,1]}}
 {id=WavePlayer_5, properties {soundWaveAssetPtr=&/Player_Grappling_Detach_Foley_01}}
 {id=WavePlayer_6, properties {soundWaveAssetPtr=&/Player_Grappling_Detach_Foley_02}}
 {id=WavePlayer_7, properties {soundWaveAssetPtr=&/Player_Grappling_Detach_Foley_03}}
 {id=WavePlayer_8, properties {soundWaveAssetPtr=&/Player_Grappling_Detach_Foley_04}}
 {id=WavePlayer_9, properties {soundWaveAssetPtr=&/Player_Grappling_Detach_Foley_05}}
]
properties {maxDistance=2097152, soundClassObject=&Fort_Gadgets_Items, volumeMultiplier=1}
