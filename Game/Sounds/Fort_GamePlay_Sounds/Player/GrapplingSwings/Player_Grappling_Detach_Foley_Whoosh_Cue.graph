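# Sound cue for the foley whoosh played when a player detaches from a grappling swing.
# A Random node picks one of five equally weighted whoosh waves and feeds it into a Modulator that varies pitch, with a lowered minimum.
# The cue plays at reduced volume through the Fort_Gadgets_Items sound class, using basic foley attenuation for spatial falloff.
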SoundCue /Game/Sounds/Fort_GamePlay_Sounds/Player/GrapplingSwings/Player_Grappling_Detach_Foley_Whoosh_Cue
root=Modulator_0
nodes [
 {id=Modulator_0, children [Random_0], properties {pitchMin=0.9}}
 {id=Random_0, children [WavePlayer_0,WavePlayer_1,WavePlayer_2,WavePlayer_3,WavePlayer_4], properties {weights [1,1,1,1,1]}}
 {
  id=WavePlayer_0
  properties {soundWaveAssetPtr=&/Player_Grappling_Deatch_Foley_Whoosh_01}
 }
 {
  id=WavePlayer_1
  properties {soundWaveAssetPtr=&/Player_Grappling_Deatch_Foley_Whoosh_02}
 }
 {
  id=WavePlayer_2
  properties {soundWaveAssetPtr=&/Player_Grappling_Deatch_Foley_Whoosh_03}
 }
 {
  id=WavePlayer_3
  properties {soundWaveAssetPtr=&/Player_Grappling_Deatch_Foley_Whoosh_04}
 }
 {
  id=WavePlayer_4
  properties {soundWaveAssetPtr=&/Player_Grappling_Deatch_Foley_Whoosh_05}
 }
]
properties {
 attenuationSettings=&Foley_Attenuation_Basic_Att
 maxDistance=1536
 soundClassObject=&Fort_Gadgets_Items
 volumeMultiplier=0.3
}
